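; Projectile fan ability for the FirePetal wandering boss, extending a parent ability with homing mask-fire projectiles.
; Holds a target-lock gameplay effect on the owner from activation until the ability ends.
; CreateFanProjectile spawns a projectile at the current internal position and rotation with randomized speed from scalable floats and an explode effect spec.
; Each projectile homes on the AI's blackboard target, unless that target carries an avoid-detection tag.
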(event Ability|EventActivateAbility
  (|CallParentFunction)
  (bind _returnvalue (Ability|ApplyGameplayEffectToOwner "/FirePetal_WanderingBoss/Abilities/GE_AbilityTargetLock_FirePetal_WanderingBoss.GE_AbilityTargetLock_FirePetal_WanderingBoss_C"))
  (Variables|Default|SetTargetLockHandle _returnvalue))

(event Ability|EventOnEndAbility
  (params (bWasCancelled :type Boolean))
  (|CallParentFunction bWasCancelled)
  (Ability|RemoveGameplayEffectFromOwnerWithHandle (Variables|Default|GetTargetLockHandle) 1))

(fn CreateFanProjectile
  (bind _loc_projectilefanproj (LocalVariables|Getloc_ProjectileFanProj))
  (if (Variables|Debug|GetbDebugAbility)
    (Rendering|Debug|DrawDebugLine (Variables|InternalData|GetCurrentProjectilePos) (+ (Variables|InternalData|GetCurrentProjectilePos) (* (Math|Vector|GetForwardVector (Variables|InternalData|GetCurrentProjectileRot)) 5e+02)) "(R=0.890625,G=0.008859,B=0.000000,A=1.000000)" 1e+01 3.0))
  (bind _returnvalue (Projectile|SpawnProjectilewithParams (Variables|Ability|GetProjectileClass) (Variables|InternalData|GetAIPawn) (Utilities|Struct|MakeSpawnProjectileParams (Variables|InternalData|GetCurrentProjectilePos) (Variables|InternalData|GetCurrentProjectileRot) (Variables|InternalData|GetCurrentProjectileRot) "(TargetSelection=(List=,bStopAtFirstSuccess=True,bKeepCheckingListOnIndestructibleHit=False,AbilityTargetBlockedGameplayTags=(GameplayTags=),bUseWeaponRanges=True,bUseMaxYawAngleToTarget=False,MaxYawAngleToTarget=0.000000),TargetGameplayEffectSpecs=,OwnerGameplayEffectSpecs=,ActivationCues=(GameplayTags=),ImpactCues=(GameplayTags=),ImpactNoiseRange=0.000000,FlyByNoiseRange=0.000000,bOverrideChargeMagnitude=False,ChargeMagnitudeOverrideValue=0.000000)" (Ability|MakeGameplayEffectContainerSpec "(TagName=\"Gameplay.Application.Explode\")") false 0 (Math|Random|RandomFloatinRange (ScalableFloat|GetValueatLevel (Variables|Projectile|GetRow_InitialSpeedMin)) (ScalableFloat|GetValueatLevel (Variables|Projectile|GetRow_InitialSpeedMax))) (Math|Random|RandomFloatinRange (ScalableFloat|GetValueatLevel (Variables|Projectile|GetRow_InitialSpeedMin)) (ScalableFloat|GetValueatLevel (Variables|Projectile|GetRow_InitialSpeedMax))) 0.0)))
  (Utilities|IsValid _returnvalue
    (:Is%20Valid
      (bind _asb_prj_fire_petal_boss_mask_fire_prj (Utilities|Casting|CastToB_Prj_FirePetal_Boss_MaskFire_Prj _returnvalue))
      (LocalVariables|Setloc_ProjectileFanProj _asb_prj_fire_petal_boss_mask_fire_prj)
      (bind _targetactor (Class|BPFLAICommonFunctions|GetTargetActorfromBlackboard (Variables|InternalData|GetAIPawn)))
      (Utilities|IsValid _targetactor
        (:Is%20Valid
          (if (GameplayTags|HasAnyMatchingGameplayTags _targetactor (Variables|Projectile|GetAvoidDetectionTagsContainer))
            (Variables|/FirePetalBosses/GameplayActors/B_Prj_FirePetal_Boss_MaskFire_Prj.B_Prj_FirePetal_Boss_MaskFire_Prj_C|SetProjHomingTarget 0 _loc_projectilefanproj)
            (else
              (Variables|/FirePetalBosses/GameplayActors/B_Prj_FirePetal_Boss_MaskFire_Prj.B_Prj_FirePetal_Boss_MaskFire_Prj_C|SetProjHomingTarget _targetactor _loc_projectilefanproj))))))))
