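; Gameplay ability that makes the Fire Petal Wandering Boss relocate by warping out, ending with the pawn hidden and force-killed.
; Activation pauses AI target focus, applies a target-lock effect until ability end, and plays a warp montage, hiding health UI and posting a killfeed message.
; A montage notify stops boss music, while teleport gameplay cues and fixed delays drive the disappearance independently of the montage.
; Both montage-selection options are the default montage, and completion handling is empty.
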(event Ability|EventActivateAbility
  (bind _returnvalue (Ability|GameplayCue|MakeGameplayCueParameters -1e+01))
  (bind _aipawn (Variables|InternalData|GetAIPawn))
  (bind _returnvalue_2 (Actor|GetComponentbyClass _aipawn "/FirePetal_WanderingBoss/Audio/Music/NPCMusicComponent_WanderingBoss.NPCMusicComponent_WanderingBoss_C"))
  (bind _aipawn_2 (Variables|InternalData|GetAIPawn))
  (bind _aipawn_3 (Variables|InternalData|GetAIPawn))
  (|CallParentFunction)
  (Variables|InternalData|SetbFireEventProcessed false)
  (Variables|InternalData|SetbEndMusicEventProcessed false)
  (CallFunction|PauseControllerFocus)
  (bind _returnvalue_3 (Ability|ApplyGameplayEffectToOwner (Variables|InternalData|GetGE_TargetLock)))
  (Variables|InternalData|SetTargetLockHandle _returnvalue_3)
  (bind _output_get (Variables|InternalData|SetSelectedWarpMontage (select (GameplayTags|HasTag (GameplayTags|GetOwnedGameplayTags (Variables|InternalData|GetAIPawn) :!class /Script/GameplayTags.BlueprintGameplayTagLibrary) (Variables|Tags|GetForceRelocateGameplayTag) true) (Variables|InternalData|GetDefaultMontage) (Variables|InternalData|GetDefaultMontage))))
  (bind (_asynctaskproxy _notifytag) (Ability|Tasks|PlayMontageAndWaitNotify "WanderingBoss_Relocate" _output_get)
    (:then
      (Variables|InternalData|SetRelocateMontageTask _asynctaskproxy)
      (Ability|GameplayEffect|RemoveLooseGameplayTags _aipawn_2 (Variables|Tags|GetForceRelocateTagContainer))
      (bind _asbp_player_pawn_fire_petal_wandering_boss (Utilities|Casting|CastToBP_PlayerPawn_FirePetal_WanderingBoss _aipawn_2))
      (Variables|/Game/Athena/AI/NPCs/Base/BP_PlayerPawn_NonParticipant_Boss.BP_PlayerPawn_NonParticipant_Boss_C|SetShowHealth false _asbp_player_pawn_fire_petal_wandering_boss)
      (Variables|/Game/Athena/AI/NPCs/Base/BP_PlayerPawn_NonParticipant_Boss.BP_PlayerPawn_NonParticipant_Boss_C|SetEnableHealthBar false _asbp_player_pawn_fire_petal_wandering_boss)
      (Class|BPPlayerPawnNonParticipantBoss|NPCDisableHealthBars _asbp_player_pawn_fire_petal_wandering_boss)
      (Widget|SetVisibility (UserInterface|GetUserWidgetObject (Variables|/Game/Athena/AI/NPCs/Base/BP_PlayerPawn_NonParticipant_Boss.BP_PlayerPawn_NonParticipant_Boss_C|GetNPCStatusWidget _asbp_player_pawn_fire_petal_wandering_boss)) "Hidden")
      (Class|BPPlayerPawnFirePetalWanderingBoss|DisplayKillfeedMessage _asbp_player_pawn_fire_petal_wandering_boss))
    (:OnCompleted
      (if (or (Variables|InternalData|GetbFireEventProcessed) (Variables|InternalData|GetbEndMusicEventProcessed))
        _
        (else
          _)))
    (:OnInterrupted
      (Development|PrintString "@@@ Event ActivateAbility [GA_Relocate_FirePetal_WanderingBoss]: WARNING - MONTAGE INTERRUPTED OR ABORTED!" false (Variables|Debug|GetbDebugAbility)))
    (:OnCancelled
      _)
    (:OnNotifyBegin
      (switch Utilities|FlowControl|Switch|SwitchonGameplayTag _notifytag
        (:Event.WanderingBoss.VoidWarp.Fire
          (Variables|InternalData|SetbFireEventProcessed true))
        (:Event.WanderingBoss.VoidWarp.EndMusic
          (Variables|InternalData|SetbEndMusicEventProcessed true)
          (Utilities|IsValid _returnvalue_2
            (:Is%20Valid
              (Class|NPCMusicComponentWanderingBoss|EndMusic _returnvalue_2)))))))
  (GameplayCue|ExecuteGameplayCue _aipawn_3 (Variables|Tags|GetTeleportRelocateGameplayCueTag))
  (GameplayCue|ExecuteGameplayCueOnActor(Burst) _aipawn_3 (Variables|Tags|GetTeleportStartGameplayTag) _returnvalue)
  (Utilities|FlowControl|Delay 0.3)
  (GameplayCue|ExecuteGameplayCueOnActor(Burst) _aipawn_3 (Variables|Tags|GetTeleportExecuteGameplayTag) _returnvalue)
  (Utilities|FlowControl|Delay 1.4)
  (Rendering|SetActorHiddenInGame _aipawn_3 true)
  (Utilities|FlowControl|Delay 0.1)
  (FortPawn|ForceKill :self _aipawn_3)
  (Ability|EndAbility))

(event Ability|EventOnEndAbility
  (params (bWasCancelled :type Boolean))
  (|CallParentFunction bWasCancelled)
  (Ability|RemoveGameplayEffectFromOwnerWithHandle (Variables|InternalData|GetTargetLockHandle) 1))

(fn PauseControllerFocus
  (bind _outfortplayerpawn (CallFunction|GetAvatarasValidFortPlayerPawn)
    (:ValidFortPlayerPawn
      (bind _asfort_athena_aibot_controller (Utilities|Casting|CastToFortAthenaAIBotController (Pawn|GetController _outfortplayerpawn)))
      (Class|FortAthenaAIBotController|PauseTargetFocusing _asfort_athena_aibot_controller))))
