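; Base sword combo ability for the FirePetal Wandering Boss, lunging at its blackboard target while a combo montage plays.
; On activation it takes AI behavior control, stops fire and applies a target-lock effect until the ability ends.
; Execution snaps to face targets within an angle threshold, then drives root-motion movement to a navmesh-raycast point capped by forward distance.
; Montage notifies feed an empty handler, and a flag allows manual execution.
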(event Ability|EventActivateAbility
  (Development|PrintString "@@@ Event ActivateAbility [GA_SwordCombo_FirePetal_WanderingBoss]: CALLED!" false (Variables|Debug|GetbDebugAbility))
  (CallFunction|InitAbility)
  (bind _returnvalue (Ability|ApplyGameplayEffectToOwner (Variables|InternalData|GetGE_TargetLock)))
  (Variables|InternalData|SetTargetLockHandle _returnvalue)
  (bind _returnvalue_2 (Ability|CommitAbility))
  (if _returnvalue_2
    (bind _targetactor (Class|BPFLAICommonFunctions|GetTargetActorfromBlackboard (Variables|InternalData|GetFortPlayerPawn)))
    (bind _output_get (Variables|InternalData|SetCurrentTargetActor _targetactor))
    (Utilities|IsValid _output_get
      (:Is%20Valid
        (if (not (Variables|TweakVars|GetbRequiresManualExecution))
          (CallFunction|EventExecuteAbility)))
      (:Is%20Not%20Valid
        (Ability|EndAbility)))
    (else
      (Ability|EndAbility))))

(event Ability|EventOnEndAbility
  (params (bWasCancelled :type Boolean))
  (bind _aibotcontroller (Variables|InternalData|GetAIBotController))
  (Development|PrintString "@@@ Event OnEndAbility [GA_SwordCombo_FirePetal_WanderingBoss]: CALLED!" false (Variables|Debug|GetbDebugAbility))
  (Ability|RemoveGameplayEffectFromOwnerWithHandle (Variables|InternalData|GetTargetLockHandle) 1)
  (Utilities|IsValid _aibotcontroller
    (:Is%20Valid
      (AI|PopBehaviorControl _aibotcontroller))))

(event Custom|Event%20ExecuteAbility
  (Development|PrintString "@@@ Event ExecuteAbility [GA_SwordCombo_FirePetal_WanderingBoss]: CALLED!" false (Variables|Debug|GetbDebugAbility))
  (Variables|InternalData|SetFortPlayerPawnLoc (Pawn|GetNavAgentLocation (Variables|InternalData|GetFortPlayerPawn)))
  (Variables|InternalData|SetCurrentTargetActorLoc (Pawn|GetNavAgentLocation (Variables|InternalData|GetCurrentTargetActor)))
  (Variables|InternalData|SetbDidRotate false)
  (bind _output_get (Variables|InternalData|SetbShouldRotate (CallFunction|ShouldUpdateRotation)))
  (if _output_get
    (bind _returnvalue (Transformation|SetActorRotation (Variables|InternalData|GetFortPlayerPawn) (Math|Rotator|FindLookatRotation (Variables|InternalData|GetFortPlayerPawnLoc) (Variables|InternalData|GetCurrentTargetActorLoc)) false))
    (Variables|InternalData|SetbDidRotate _returnvalue))
  (bind _hitlocation (AI|Navigation|NavigationRaycast (Variables|InternalData|GetFortPlayerPawnLoc) (Math|Vector|SelectVector (+ (Variables|InternalData|GetFortPlayerPawnLoc) (* (Math|Vector|GetForwardVector (Math|Rotator|FindLookatRotation (Variables|InternalData|GetFortPlayerPawnLoc) (Variables|InternalData|GetCurrentTargetActorLoc))) (Math|Float|Clamp(Float) (Math|Vector|Distance2D(Vector) (Variables|InternalData|GetFortPlayerPawnLoc) (Variables|InternalData|GetCurrentTargetActorLoc)) 0.0 (ScalableFloat|GetValueatLevel (Variables|TweakVars|GetSF_MaxForwardDist))))) (+ (Variables|InternalData|GetFortPlayerPawnLoc) (* (Transformation|GetActorForwardVector (Variables|InternalData|GetFortPlayerPawn)) (ScalableFloat|GetValueatLevel (Variables|TweakVars|GetSF_MaxForwardDist)))) (and (Variables|InternalData|GetbShouldRotate) (Variables|InternalData|GetbDidRotate))) "/Game/Athena/AI/NavFilters/NavFilter_NPC_AllNavmesh.NavFilter_NPC_AllNavmesh_C" (Variables|InternalData|GetAIBotController)))
  (bind _asynctaskproxy (Ability|Tasks|ApplyRootMotionMoveToForce "WanderingBoss_SwordAttackForce" _hitlocation (ScalableFloat|GetValueatLevel (Variables|TweakVars|GetSF_MoveToTime)) false "MOVE_None" false 0 "SetVelocity"))
  (Variables|InternalData|SetMoveToTask _asynctaskproxy)
  (if (not (Variables|TweakVars|GetbRequiresManualExecution))
    (CallFunction|EventStartComboMontage)))

(event Custom|Event%20StartComboMontage
  (bind (_montageinfo_montagetoplay _montageinfo_animplayrate _montageinfo_animrootmotiontranslationscale) (Utilities|Struct|BreakFortGameplayAbilityMontageInfo (Variables|GetMontageInfo_MontageToPlay)))
  (Development|PrintString "@@@ Event StartComboMontage [GA_SwordCombo_FirePetal_WanderingBoss]: CALLED!" false (Variables|Debug|GetbDebugAbility))
  (bind (_asynctaskproxy _notifytag) (Ability|Tasks|PlayMontageAndWaitNotify "WanderingBoss_SwordAttackAnim" _montageinfo_montagetoplay _montageinfo_animplayrate "None" true _montageinfo_animrootmotiontranslationscale)
    (:then
      (Variables|InternalData|SetAnimTask _asynctaskproxy))
    (:OnCompleted
      _)
    (:OnInterrupted
      _)
    (:OnCancelled
      _)
    (:OnNotifyBegin
      (CallFunction|HandleMontageNotify _notifytag)))
  (Ability|EndAbility))

(fn HandleMontageNotify
  (params (NotifyTag :type Gameplay%20Tag%20Structure)))

(fn InitAbility
  (bind _output_get (Variables|InternalData|SetFortPlayerPawn (CallFunction|GetAvatarasFortPlayerPawn)))
  (bind _asfort_athena_aibot_controller (Utilities|Casting|CastToFortAthenaAIBotController (Pawn|GetController _output_get))
    (:then
      (bind _output_get_2 (Variables|InternalData|SetAIBotController _asfort_athena_aibot_controller))
      (AI|PushBehaviorControl _output_get_2 -2097152)
      (Class|FortAthenaAIBotController|StopFire (Variables|InternalData|GetAIBotController)))
    (:CastFailed
      (Ability|EndAbility))))

(fn ShouldUpdateRotation
  (bind _degrees (Class|BPFLAICommonFunctions|GetAngleDegreestoTarget (Variables|InternalData|GetFortPlayerPawn) (Pawn|GetNavAgentLocation (Variables|InternalData|GetCurrentTargetActor))))
  (return (<= _degrees (ScalableFloat|GetValueatLevel (Variables|TweakVars|GetSF_RotateFacingThreshold)))))
