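; Finisher ability for the FirePetal wandering boss's sword combo, activatable only when the AI's blackboard target is inside its leash.
; Activation starts the combo montage and adds looping charge cues, which are removed at the active hit or on ability end.
; The MoveTo notify executes the ability and defers to the parent.
; The Active notify runs async targeting with a Masamune full-charge preset, applying the finisher damage effect to the results.
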(event Ability|EventActivateAbility
  (|CallParentFunction)
  (CallFunction|EventStartComboMontage)
  (Projectile|AddGameplayCue%28Looping%29Local (Variables|InternalData|GetFortPlayerPawn) (Variables|Tags|GetGC_Charging) (Ability|GameplayCue|MakeGameplayCueParameters 0.0))
  (GameplayCue|AddGameplayCueOnActor(Looping) (Variables|InternalData|GetFortPlayerPawn) (Variables|Tags|GetGC_AV_Charging) (Ability|GameplayCue|MakeGameplayCueParameters 0.0)))

(event EventHandleMontageNotify
  (params (NotifyTag :type Gameplay%20Tag%20Structure))
  (switch Utilities|FlowControl|Switch|SwitchonGameplayTag NotifyTag
    (:Event.WanderingBoss.SwordCombo.MoveTo.01
      (CallFunction|EventExecuteAbility)
      (|CallParentFunction NotifyTag))
    (:Event.WanderingBoss.SwordCombo.Active.01
      (CallFunction|RemoveChargeCues)
      (bind (_asynctaskproxy _targetinghandle) (Class|AsyncActionPerformTargeting|PerformTargetingAsyncAction (Variables|InternalData|GetFortPlayerPawn) "/MasamuneGameplay/Gameplay/Targeting/TP_Masamune_FullCharge.TP_Masamune_FullCharge")
        (:Targeted
          (Ability|Tasks|ApplyGameplayEffectToHitResultsAsync (Ability|GetActivatingAbilityComponent) (TargetingSystem|TargetingResults|GetTargetingResults (GameInstanceSubsystems|GetTargetingSubsystem) _targetinghandle) "/FirePetal_WanderingBoss/Abilities/GE_SwordFinisher_Damage_FirePetal_WanderingBoss.GE_SwordFinisher_Damage_FirePetal_WanderingBoss_C" 0 1 (Class|AthenaFunctionLibrary|MakeSimpleFortEffectContext (Variables|InternalData|GetFortPlayerPawn) (Variables|InternalData|GetFortPlayerPawn))))))))

(event Ability|EventOnEndAbility
  (params (bWasCancelled :type Boolean))
  (|CallParentFunction bWasCancelled)
  (CallFunction|RemoveChargeCues))

(fn CanActivateAbility
  (params (ActorInfo :type Gameplay%20Ability%20Actor%20Info%20Structure) (Handle :type Gameplay%20Ability%20Spec%20Handle%20Structure))
  (bind _returnvalue (CallFunction|GetAvatarasFortPlayerPawn))
  (bind _relevanttags (|CallParentFunction ActorInfo Handle))
  (bind _targetactor (Class|BPFLAICommonFunctions|GetTargetActorfromBlackboard _returnvalue))
  (bind _asfort_athena_aibot_controller (Utilities|Casting|CastToFortAthenaAIBotController (Pawn|GetController _returnvalue))
    (:then
      (return _relevanttags (Class|FortAthenaAIBotController|IsLocationInsideLeash _asfort_athena_aibot_controller (Pawn|GetNavAgentLocation _targetactor) false 0.0)))
    (:CastFailed
      (return _relevanttags false))))

(fn RemoveChargeCues
  (Projectile|RemoveGameplayCue%28Looping%29Local (Variables|InternalData|GetFortPlayerPawn) (Variables|Tags|GetGC_Charging) (Ability|GameplayCue|MakeGameplayCueParameters 0.0))
  (GameplayCue|RemoveGameplayCueOnActor(Looping) (Variables|InternalData|GetFortPlayerPawn) (Variables|Tags|GetGC_AV_Charging) (Ability|GameplayCue|MakeGameplayCueParameters 0.0)))
