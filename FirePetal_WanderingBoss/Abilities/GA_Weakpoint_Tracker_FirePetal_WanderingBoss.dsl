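; Gameplay ability for the FirePetal Wandering Boss that detects hits on weakpoint bones and triggers feedback.
; On activation it caches the pawn and AI controller, then binds to the boss's server damage event.
; Each hit's bone name is compared to a tunable bone array; a match executes a gameplay cue on the owner, smuggling the array index through GameplayEffectLevel.
; ValidateExceptions can bypass the bone match but currently always returns false.
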(event Ability|EventActivateAbility
  (Ability|CommitAbility)
  (CallFunction|InitAbility)
  (Default|BindEventtoOnServerDamagedWanderingBoss (Variables|InternalData|GetWanderingBossPawn) (EventDispatchers|CreateEvent :!event Custom|DamagedOnServer)))

(event Custom|DamagedOnServer
  (params (DamageTags :type Gameplay%20Tag%20Container%20Structure) (HitResult :type Hit%20Result%20Structure) (Instigator :type Controller%20Object%20Reference) (DamageCauser :type Actor%20Object%20Reference))
  (bind (_bblockinghit _binitialoverlap _time _distance _location _impactpoint _normal _impactnormal _physmat _hitactor _hitcomponent _hitbonename) (Collision|BreakHitResult HitResult))
  (for (_array_element _array_index) (Variables|tunables|GetWeakpointBoneArray)
    (if (or (== _array_element _hitbonename) (CallFunction|ValidateExceptions DamageCauser DamageTags))
      (CallFunction|ProcessHitFeedback _array_index))))

(event Ability|EventOnEndAbility
  (params (bWasCancelled :type Boolean)))

(fn InitAbility
  (bind _outfortplayerpawn (CallFunction|GetAvatarasValidFortPlayerPawn)
    (:ValidFortPlayerPawn
      (bind _output_get (Variables|InternalData|SetAIPawn _outfortplayerpawn))
      (bind _asfort_athena_aibot_controller (Utilities|Casting|CastToFortAthenaAIBotController (Pawn|GetController _output_get))
        (:then
          (Variables|InternalData|SetAIController _asfort_athena_aibot_controller)
          (bind _asbp_player_pawn_fire_petal_wandering_boss (Utilities|Casting|CastToBP_PlayerPawn_FirePetal_WanderingBoss (Variables|InternalData|GetAIPawn))
            (:then
              (Variables|InternalData|SetWanderingBossPawn _asbp_player_pawn_fire_petal_wandering_boss))
            (:CastFailed
              (Ability|EndAbility)))
          (return))
        (:CastFailed
          _)))
    (:AvatarCastFailed
      _)))

(fn ProcessHitFeedback
  (params (WeakpointBoneArrayIndex :type Integer))
  (Ability|ExecuteGameplayCueWithParamsOnOwner (Variables|Tunables|GetWeakpointGameplayCue) (Ability|GameplayCue|MakeGameplayCueParameters :NormalizedMagnitude 0.0 :RawMagnitude 0.0 :MatchedTagName (Variables|Tunables|GetWeakpointGameplayCue) :Location "0,0,0" :Normal "0,0,0" :GameplayEffectLevel WeakpointBoneArrayIndex :AbilityLevel 1 :bReplicateLocationWhenUsingMinimalRepProxy false))
  (return))

(fn ValidateExceptions
  (params (DamageCauser :type Actor%20Object%20Reference) (DamageTags :type Gameplay%20Tag%20Container%20Structure))
  (return false))
